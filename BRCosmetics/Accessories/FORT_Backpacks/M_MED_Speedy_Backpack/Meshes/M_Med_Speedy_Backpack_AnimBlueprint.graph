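# Animation Blueprint for the Speedy backpack cosmetic, opening its five flaps when the pawn moves fast.
# Flap bones rotate via PawnSpeed2D alphas, using zero-width speed ranges as staggered thresholds with interpolation quick to rise and slow to fall.
# The update event reads aim pitch from the owner's FortPlayerAnimInstance, scales it and caps it at zero.
# That AimOffset feeds an Apply Limits node on backpack_root, followed by a fixed additive roll.
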AnimBlueprint /BRCosmetics/Accessories/FORT_Backpacks/M_MED_Speedy_Backpack/Meshes/M_Med_Speedy_Backpack_AnimBlueprint
target_skeleton=&/M_Med_Speedy_Backpack_Skeleton
graphs [
 AnimGraph {nodes [
   AnimGraphNode_SubInput_0 = Input Pose()
   n1 = Apply Limits!(
    ComponentPose:Component Space Pose Link Structure=Local To Component(LocalPose=@AnimGraphNode_SubInput_0)
    Alpha:float=1
    AngularOffsets_0:Vector=Get AimOffset() -> (AimOffset:Vector)
   ) -> (Pose:Component Space Pose Link Structure) {
    is=title AnimGraphNode_ApplyLimits
    alphaBoolBlend {blendOption=Linear}
    angularOffsets [<0 0 0>]
    angularRangeLimits [{bone {boneName=backpack_root}, limitMax=<0 0 20>, limitMin=<0 0 0>}]
   }
   n2 = "Transform (Modify) Bone - Bone: backpack_root"(ComponentPose=@n1, Alpha=1, Translation="0,0,0", Rotation="0.000,0.000,12", Scale="1,1,1") {rotationMode=BMM_Additive, alphaBoolBlend {blendOption=Linear}, boneToModify {boneName=backpack_root}, rotation {roll=12}}
   n3 = Get PawnSpeed2D()
   n4 = "Transform (Modify) Bone - Bone: flap_mid"!(ComponentPose=@n2, Alpha=@n3, Translation="0,0,0", Scale="1,1,1") -> (Pose) {
    lODThreshold=1
    rotationMode=BMM_Replace
    rotationSpace=BCS_BoneSpace
    alphaBoolBlend {blendOption=Linear}
    alphaScaleBiasClamp {bInterpResult=true, bMapRange=true, interpSpeedDecreasing=1, interpSpeedIncreasing=18, inRange {max=430, min=430}}
    boneToModify {boneName=flap_mid}
    rotation {yaw=45.5595}
   }
   n5 = "Transform (Modify) Bone - Bone: flap_top_l"!(ComponentPose=@n4, Alpha=@n3, Translation="0,0,0", Scale="1,1,1") -> (Pose) {
    lODThreshold=1
    rotationMode=BMM_Replace
    rotationSpace=BCS_BoneSpace
    alphaBoolBlend {blendOption=Linear}
    alphaScaleBiasClamp {bInterpResult=true, bMapRange=true, interpSpeedDecreasing=1, interpSpeedIncreasing=18, inRange {max=450, min=450}}
    boneToModify {boneName=flap_top_l}
    rotation {yaw=42.1327}
   }
   n6 = "Transform (Modify) Bone - Bone: flap_top_r"!(ComponentPose=@n5, Alpha=@n3, Translation="0,0,0", Scale="1,1,1") -> (Pose) {
    lODThreshold=1
    rotationMode=BMM_Replace
    rotationSpace=BCS_BoneSpace
    alphaBoolBlend {blendOption=Linear}
    alphaScaleBiasClamp {bInterpResult=true, bMapRange=true, interpSpeedDecreasing=1, interpSpeedIncreasing=18, inRange {max=450, min=450}}
    boneToModify {boneName=flap_top_r}
    rotation {yaw=42.1327}
   }
   n7 = "Transform (Modify) Bone - Bone: flap_bott_l"!(ComponentPose=@n6, Alpha=@n3, Translation="0,0,0", Scale="1,1,1") -> (Pose) {
    lODThreshold=1
    rotationMode=BMM_Replace
    rotationSpace=BCS_BoneSpace
    alphaBoolBlend {blendOption=Linear}
    alphaScaleBiasClamp {bInterpResult=true, bMapRange=true, interpSpeedDecreasing=1, interpSpeedIncreasing=18, inRange {max=415, min=415}}
    boneToModify {boneName=flap_bott_l}
    rotation {yaw=44.0238}
   }
   n8 = "Transform (Modify) Bone - Bone: flap_bott_r"!(ComponentPose=@n7, Alpha=@n3, Translation="0,0,0", Scale="1,1,1") -> (Pose) {
    lODThreshold=1
    rotationMode=BMM_Replace
    rotationSpace=BCS_BoneSpace
    alphaBoolBlend {blendOption=Linear}
    alphaScaleBiasClamp {bInterpResult=true, bMapRange=true, interpSpeedDecreasing=1, interpSpeedIncreasing=18, inRange {max=415, min=415}}
    boneToModify {boneName=flap_bott_r}
    rotation {yaw=44.0238}
   }
   Output Pose(Result=Component To Local(ComponentPose=@n8)) {layerGroup=None}
  ]}
 EventGraph {nodes [
   Event BlueprintUpdateAnimation()
   > n1 = Cast To FortPlayerAnimInstance(Object=GetAnimInstance(self=GetOwnerSkeletalMeshComponent()))
   n2 = Get!(self:Fort Player Anim Instance Object Reference=@n1) -> (AdjustedAim_ResultingYaw:float=~, AdjustedAim_ResultingPitch:float=~, AdjustedAim_ZeroOutPitchWeightCurveName:Name)
   n3 = Get AimOffsetMultiplier() -> (AimOffsetMultiplier:double)
   > Set!(
    execute
    AimOffset_X:double=0
    AimOffset_Y:double=0
    AimOffset_Z:double="Min (Float)"(A:double=float * float(A=@n2.AdjustedAim_ResultingPitch, B=@n3), B:double=0) -> (ReturnValue:double) {is=title K2Node_CommutativeAssociativeBinaryOperator}
   ) -> (then, Output_Get:Vector)
   TryGetPawnOwner()
  ]}
]
variables [
 AimOffset {type=Vector}
 AimOffsetMultiplier {type=number}
]
